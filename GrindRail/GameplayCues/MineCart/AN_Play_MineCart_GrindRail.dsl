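; Timed effect anim notify for grind rail mine cart visuals, tailoring its spawned effect to the rail being ridden.
; On notify begin, it finds the mesh owner's grinding provider component and casts its current rail actor to BGA_GrindRail.
; The rail's minecart-variant flag sets whether the effect is active, with reset enabled.
; The rail's sprint color hue and saturation feed the effect's float parameters.
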(fn Received_NotifyBegin
  (params (MeshComp :type Skeletal%20Mesh%20Component%20Object%20Reference) (Animation :type Anim%20Sequence%20Base%20Object%20Reference) (TotalDuration :type Float%20%28single-precision%29) (EventReference :type Anim%20Notify%20Event%20Reference%20Structure%20%28by%20ref%29))
  (bind _returnvalue (Components|GetOwner MeshComp))
  (bind _returnvalue_2 (Actor|GetComponentbyClass _returnvalue "/GrindRail/Blueprints/BP_FortPawnGrindingProvider.BP_FortPawnGrindingProvider_C"))
  (Utilities|IsValid _returnvalue_2
    (:Is%20Valid
      (bind _asbga_grind_rail (Utilities|Casting|CastToBGA_GrindRail (Class|FortPawnComponentGrindingProvider|GetCurrentGrindingRailActor _returnvalue_2)))
      (bind _returnvalue_3 (AnimNotify|GetSpawnedEffect MeshComp))
      (Components|Activation|SetActive _returnvalue_3 (Variables|/GrindRail/Blueprints/BGA_GrindRail.BGA_GrindRail_C|GetblsMinecartVariant _asbga_grind_rail) true)
      (Effects|Components|ParticleSystem|SetFloatParameter _returnvalue_3 "SprintColor_Hue" (Variables|/GrindRail/Blueprints/BGA_GrindRail.BGA_GrindRail_C|GetSprintColor_Hue _asbga_grind_rail))
      (Effects|Components|ParticleSystem|SetFloatParameter _returnvalue_3 "SprintColor_Saturation" (Variables|/GrindRail/Blueprints/BGA_GrindRail.BGA_GrindRail_C|GetSprintColor_Saturation _asbga_grind_rail))
      (return false))))
